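# Niagara ribbon trail for the STW006 Tier 7 pickaxe cosmetic, drawing a short-lived blue animation trail from a skeletal mesh.
# Its single emitter uses shared AnimTrail functions to spawn particles, initialize bone indices and set width, color, lifetime and offset.
# Each update follows the skeletal mesh and applies ribbon material color and erosion.
# A small initial upward velocity also supplies the ribbon facing vector.
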NiagaraSystem /BRCosmetics/Weapons/FORT_Melee/STW006_Tier_7/NS_Pickaxe_AnimTrail_STW006_Tier_7
dependencies {
 usedDataInterfaces []
 usedDynamicInputs []
 usedModules [
  &self:SystemSpawnScript.SystemScriptSource.SystemScriptGraph.NiagaraNodeAssignment_1.SetVariables_#1
  &SystemState
 ]
}
emitters {Em_AnimTrail {
  emitterUpdateScript [
   &EmitterState()
   &FN_SpawnAnimTrail()
  ]
  particleSpawnScript [
   &FN_InitializeAnimTrailBoneIndecies()
   &FN_InitializeAnimTrail(AnimTrailWidth=60, Color=<rgba 0 0.4231 1 1>, Lifetime=0.3, Position Offset=<0 0 25>)
   &AddVelocity(Velocity=<0 0 1>)
  ]
  particleUpdateScript [
   &ParticleState()
   /SetVariables(Particles.RibbonFacing=$Particles.Initial.Velocity) {isSetParameters=true}
   &FN_UpdateAnimTrail(SkeletalMesh=$System.Skeletal Mesh)
   &FN_RibbonMaterial(Color=<rgba 0 0.2949 1 1>, Opacity Erosion=1)
   &SolveForcesAndVelocity(Clamp Velocity=false, Limit Acceleration=false)
  ]
  rendererClasses [&NiagaraRibbonRendererProperties]
  renderers [{rendererClass=&NiagaraRibbonRendererProperties}]
 }}
summary {
 systemName=NS_Pickaxe_AnimTrail_STW006_Tier_7
 emitters [{emitterName=Em_AnimTrail, rendererClasses [&NiagaraRibbonRendererProperties]}]
 userVariables [
  User.RibbonIDOffset {type=NiagaraInt32}
  User.SkeletalMesh {
   type=NiagaraDataInterfaceSkeletalMesh
   defaultValue {
    struct=NiagaraExt_VariableValue_DataInterface
    value {dataInterface=&self:NiagaraDataInterfaceSkeletalMesh_0}
   }
  }
 ]
}
